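# Niagara vehicle boost cosmetic "Sticky 02", a pumpkin-themed exhaust effect with green, orange and dark smoke colors.
# It layers a multi-renderer ribbon trail, looping aura and glimpse bursts split by spawn group, and spinning pumpkin mesh projectiles with fire sprites.
# Pumpkins emit death events, while the explosion flipbook emitter has no active spawn module shown.
# User parameters set boost position and car offset, frontend-preview switches retune values, and painted-variant overrides recolor trail and pumpkins.
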NiagaraSystem /VehicleCosmetics/Boosts/Boost_Sticky02/FX/NS_Sticky_02
dependencies {
 usedDataInterfaces [
  &NiagaraDataInterfaceCurve
  &NiagaraDataInterfaceArrayFloat
  &NiagaraDataInterfaceVectorCurve
  &NiagaraDataInterfaceArrayFloat2
  &NiagaraDataInterfaceVector2DCurve
  &NiagaraDataInterfaceArrayFloat3
  &NiagaraDataInterfaceSpriteRendererInfo
  &NiagaraDataInterfaceSkeletalMesh
 ]
 usedDynamicInputs []
 usedModules [
  &self:SystemSpawnScript.SystemScriptSource.SystemScriptGraph.NiagaraNodeAssignment_0.SetVariables_#1
  &SystemState
 ]
}
emitters {
 NE_VC_Boost_Core_Aura {
  emitterUpdateScript [
   &EmitterState(Inactive Response=::NewEnumerator1, Life Cycle Mode=::NewEnumerator1, Loop Duration=0.09, UseLoopDelay=false)
   &SpawnBurst_Instantaneous(Spawn Group=1, Use Spawn Probability=false, UseLoopCountLimit=false) {module=SpawnBurst_Instantaneous001}
   &SpawnBurst_Instantaneous(Spawn Count=4, Use Spawn Probability=false, UseLoopCountLimit=false)
   &VC_FilteredSocketCount()
   &SpawnRate(SpawnRate=50, Use Spawn Probability=false) {enabled=false}
  ]
  particleSpawnScript [
   &VC_PaintedVariantsOverride() {enabled=false}
   /SetVariables(Particles.Color2=<rgba 0.08 0.0651 0.024 1>, Particles.RENDERTAG:NiagaraInt32=&SpawnGroup()) {isSetParameters=true}
   &InitializeParticle(
    Color=<rgba 0.5 1 0 1>
    Lifetime=/VehicleCosmetics/FX/Niagara/Modules/MergeEmitters/VC_FloatBySpawnGroup(
     SpawnGroup or ExecIndex:NiagaraInt32=0
     CustomInt:NiagaraInt32=0
     MaximumSpawnGroups:NiagaraInt32=2
     Float SG 0:NiagaraFloat=1
     Float SG 1:NiagaraFloat=&RandomRangeFloat(Minimum=0.36, Maximum=0.6)
     Float SG 2:NiagaraFloat=0
     Float SG 3:NiagaraFloat=0
     Float SG 4:NiagaraFloat=0
     Float SG 5:NiagaraFloat=0
    )
    Position Offset=&AddVectorToPosition(
     Position=&AddVectorToPosition(
      Position=/VehicleCosmetics/FX/Niagara/Modules/MergeEmitters/VC_VectorBySpawnGroup(
       SpawnGroup or ExecIndex:NiagaraInt32=0
       CustomInt:NiagaraInt32=0
       MaximumSpawnGroups:NiagaraInt32=2
       Vector Group 0:Vector3f=<-130 0 0>
       Vector Group 1:Vector3f=<-50 0 0>
       Vector Group 2:Vector3f=<0 0 0>
       Vector Group 3:Vector3f=<0 0 0>
       Vector Group 4:Vector3f=<0 0 0>
       Vector Group 5:Vector3f=<0 0 0>
      )
      Vector=$User.SingleBoostPos
     )
     Vector=&Multiply_VectorByFloat(Vector=$User.DTCarOffset, Float=1)
    )
    Sprite Rotation Angle=90
    Sprite Rotation Mode=::NewEnumerator0
    Sprite Size Max=&RandomRangeVector2D(Minimum=<120 60>, Maximum=<120 70>)
    Sprite Size Min=&RandomRangeVector2D(Minimum=<80 50>, Maximum=<80 40>)
    Sprite Size Mode=::NewEnumerator2
    UsePositionOffset=true
    Write Material Random=true
   )
   &PartitionParticles(Number Of Partitions=$Emitter.PortCount) {enabled=false}
   &SkeletalMeshLocation(
    Bone / Socket Index=$Particles.PartitionParticles.Partition
    Bone / Socket Sampling Mode=::NewEnumerator5
    MaskSpawnGroup=false
    Mesh Sampling Type=::NewEnumerator6
    Skeletal Mesh=$System.SkeletalMesh
   ) {enabled=false}
   &ApplyInitialForces()
   &AddVelocity(
    Cone Angle=5
    Cone Axis=<-1 0 0>
    Module.UseConeFalloff=false
    Module.UseVelDistribution=false
    Override Seed=false
    Velocity Mode=::NewEnumerator2
    Velocity Speed=&VC_FloatBySpawnGroup(
     SpawnGroup or ExecIndex:NiagaraInt32=0
     CustomInt:NiagaraInt32=0
     MaximumSpawnGroups:NiagaraInt32=2
     Float SG 0:NiagaraFloat=0
     Float SG 1:NiagaraFloat=&RandomRangeFloat(Minimum=250, Maximum=220)
     Float SG 2:NiagaraFloat=0
     Float SG 3:NiagaraFloat=0
     Float SG 4:NiagaraFloat=0
     Float SG 5:NiagaraFloat=0
    )
   )
   &AlignSpriteToMeshOrientation(
    Mesh Orientation Relative Sprite Alignment Vector=&RandomRangeVector(Minimum=<1 0 0>, Maximum=<1 0 0>)
    Mesh Orientation Relative Sprite Facing Vector:Vector3f=<0 1 1>
    SetAlignment=true
    SetFacing=false
   ) {module=AlignSpriteToMeshOrientation001}
   &InitialMeshOrientation(
    Mesh Axis To Orient=<0 1 0>
    Mesh Orientation Mode=::NewEnumerator2
    Orientation Coordinate Space=::Local
    Orientation Vector=&SelectVectorFromArray(
     Array Sampling Mode=::NewEnumerator0
     Direct Array Index=?
     Vector Selection Array=DataInterface() {propertyValues {FloatData [<0 0 1>, <0 1 0>]}}
     Random Seed=0
    )
    Use Rotation Vector=true
   )
   &AlignSpriteToMeshOrientation(
    Mesh Orientation Relative Sprite Alignment Vector=&RandomRangeVector(Minimum=<1 0 0>, Maximum=<1 0 0>)
    Mesh Orientation Relative Sprite Facing Vector:Vector3f=<0 1 1>
    SetAlignment=true
    SetFacing=false
   )
   &ShapeLocation(
    Box / Plane Mode:ENiagara_BoxPlaneMode=::NewEnumerator1
    MaskSpawnGroup=false
    Override Seed=false
    Plane Size:Vector2f=&MaskVector2dBySpawnGroup(Masked Vector2D=<0 0>, Passthrough Vector2D=<10 5>, Spawn Group Mask=0)
    Rotation Coordinate Space=::Simulation
    Rotation Mode=::NewEnumerator1
    Shape Primitive=::NewEnumerator2
    UseOwnerScale=true
   )
   &CameraOffset(Camera Offset Amount=&VC_FloatBySpawnGroup(
     SpawnGroup or ExecIndex:NiagaraInt32=0
     CustomInt:NiagaraInt32=0
     MaximumSpawnGroups:NiagaraInt32=2
     Float SG 0:NiagaraFloat=55
     Float SG 1:NiagaraFloat=30
     Float SG 2:NiagaraFloat=0
     Float SG 3:NiagaraFloat=0
     Float SG 4:NiagaraFloat=0
     Float SG 5:NiagaraFloat=0
    ))
  ]
  particleUpdateScript [
   &ParticleState()
   &ScaleSpriteSize(
    Non-Uniform Curve Scale=&RandomRangeVector2D(Minimum=<0.8 1.2>, Maximum=<0.75 1.1>)
    Non-Uniform Curve Sprite Scale=DataInterface() {propertyValues {
      XCurve {keys [
        {arriveTangent=-0.0309, interpMode=RCIM_Cubic, leaveTangent=-0.0309, tangentMode=RCTM_User, time=0, value=0.25}
        {arriveTangentWeight=0.2668, interpMode=RCIM_Cubic, leaveTangentWeight=0.0667, tangentMode=RCTM_User, time=0.5, value=1.2}
        <at 0.8 0.75 RCIM_Cubic>
       ]}
      YCurve {keys [
        {arriveTangent=1.4884, interpMode=RCIM_Cubic, leaveTangent=1.4884, tangentMode=RCTM_User, time=0, value=0.25}
        {arriveTangent=0.9375, interpMode=RCIM_Cubic, leaveTangent=0.9375, time=0.35, value=1.1}
        <at 0.8 1 RCIM_Cubic>
       ]}
     }}
    Scale Sprite Size Mode=::NewEnumerator3
   )
   &UpdateMeshOrientation(Rotation Rate=&RandomRangeFloat(Minimum=0.1, Maximum=-0.1, Evaluation Type=::NewEnumerator1))
   &SolveForcesAndVelocity(Clamp Velocity=false, Limit Acceleration=false)
   &ScaleColor(
    Scale Alpha=&FloatFromCurve(FloatCurve=DataInterface() {propertyValues {Curve {keys [
         {interpMode=RCIM_Cubic, tangentMode=RCTM_User, time=0.1, value=0}
         {interpMode=RCIM_Cubic, tangentMode=RCTM_User, time=0.5, value=1}
         {tangentMode=RCTM_User, time=1, value=0}
        ]}}})
    ScaleA=true
    ScaleRGB=true
   )
   /SetVariables(Particles.SPRITESCOLOR:LinearColor=&VC_DynamicScaleColor(
     Color to Scale:LinearColor=$Particles.Initial.Color2
     RGB Scale:Vector3f=&Multiply_VectorByFloat(
      Vector=<1 1 1>
      Float=&FloatFromCurve(
       FloatCurve=DataInterface() {propertyValues {Curve {keys [
           {arriveTangent=0.8043, interpMode=RCIM_Cubic, leaveTangent=0.8043, tangentMode=RCTM_User, time=0, value=0.6}
           {arriveTangent=2.6893, interpMode=RCIM_Cubic, leaveTangent=-1.0064, tangentMode=RCTM_Break, time=0.1, value=0.8}
           {tangentMode=RCTM_User, time=1, value=0.4}
          ]}}}
       Scale Curve=&MakeCustomFloatFromBool(
        Bool=$User.bIsFrontendPreview
        True Float=&RandomRangeFloat(Minimum=7, Maximum=4)
        False Float=&RandomRangeFloat(Minimum=10, Maximum=20)
       )
      )
     )
     Alpha Scale:NiagaraFloat=&FloatFromCurve(FloatCurve=DataInterface() {propertyValues {Curve {keys [
          {interpMode=RCIM_Cubic, tangentMode=RCTM_User, time=0, value=1}
          {interpMode=RCIM_Cubic, tangentMode=RCTM_User, time=0.5, value=1}
          {arriveTangent=-2.1126, tangentMode=RCTM_User, time=1, value=0}
         ]}}})
    )) {isSetParameters=true}
   &SubUVAnimation(
    Random Start Frame=true
    "Sprite Renderer "=DataInterface() {propertyValues {SpriteRenderer=&self:NE_VC_Boost_Core_Aura_2.NiagaraSpriteRendererProperties_1}}
    Start Frame Offset=4
    SubUV Animation Mode:ENiagara_SubUVLookupModeV2=::NewEnumerator1
    SubUV Curve=DataInterface() {propertyValues {Curve {keys [<at 0 1 RCIM_Cubic>,<at 1 0 RCIM_Cubic>]}}}
    UseEndFrame=false
    UseStartFrame=true
   )
  ]
  rendererClasses [&NiagaraMeshRendererProperties, &NiagaraSpriteRendererProperties]
  renderers [
   {rendererClass=&NiagaraMeshRendererProperties, rendererIndex=0}
   {rendererClass=&NiagaraSpriteRendererProperties, rendererIndex=1}
  ]
 }
 NE_VC_Boost_Explosions003 {
  emitterUpdateScript [
   &EmitterState()
   &VC_FilteredSocketCount()
   &SpawnRate(SpawnRate=50, Use Spawn Probability=false) {enabled=false}
  ]
  particleSpawnScript [
   &VC_PaintedVariantsOverride() {enabled=false}
   &InitializeParticle(
    Color=&MakeCustomLinearColorFromBool(Bool=$User.bIsFrontendPreview, True Color=<rgba 1 0.25 0 1>, False Color=<rgba 1.5 0.375 0 1>)
    Lifetime=&MakeCustomFloatFromBool(
     Bool=$User.bIsFrontendPreview
     True Float=&RandomRangeFloat(Minimum=0.2, Maximum=0.25)
     False Float=0.5
    )
    Sprite Rotation Mode=::NewEnumerator2
    Sprite Size Mode=::NewEnumerator3
    Uniform Sprite Size=&MakeCustomFloatFromBool(
     Bool=$User.bIsFrontendPreview
     True Float=&RandomRangeFloat(Minimum=50, Maximum=40)
     False Float=&RandomRangeFloat(Minimum=55, Maximum=45)
    )
    UsePositionOffset=false
    Write Material Random=true
   )
   &PartitionParticles(Number Of Partitions=$Emitter.PortCount)
   &SkeletalMeshLocation(
    Bone / Socket Index=$Particles.PartitionParticles.Partition
    Bone / Socket Sampling Mode=::NewEnumerator5
    MaskSpawnGroup=false
    Mesh Sampling Type=::NewEnumerator6
    Skeletal Mesh=$System.SkeletalMesh
   )
   &AddVelocity(Velocity=<-1500 0 0>) {enabled=false}
   &CameraOffset(Camera Offset Amount=&MakeCustomFloatFromBool(Bool=$User.bIsFrontendPreview, True Float=0, False Float=0))
  ]
  particleUpdateScript [
   &ParticleState()
   &ScaleSpriteSize(
    Scale Sprite Size Mode=::NewEnumerator0
    Uniform Scale Factor=&MakeCustomFloatFromBool(
     Bool=$User.bIsFrontendPreview
     True Float=&FloatFromCurve(FloatCurve=DataInterface() {propertyValues {Curve {keys [
          <at 0 1 RCIM_Cubic>
          {arriveTangent=5.7975, interpMode=RCIM_Cubic, leaveTangent=-0.5607, tangentMode=RCTM_Break, time=0.7, value=2.5}
          {arriveTangentWeight=0.9175, interpMode=RCIM_Cubic, tangentMode=RCTM_User, time=1, value=2.5}
         ]}}})
     False Float=&FloatFromCurve(
      FloatCurve=DataInterface() {propertyValues {Curve {keys [<at 0 1 RCIM_Cubic>, {arriveTangent=6.8809, interpMode=RCIM_Cubic, leaveTangent=6.8809, tangentMode=RCTM_User, time=1, value=3}]}}}
      Scale Curve=1.5
     )
    )
   ) {module=ScaleSpriteSize001}
   &SolveForcesAndVelocity(Clamp Velocity=false, Limit Acceleration=false)
   &SubUVAnimation(
    "Sprite Renderer "=DataInterface() {propertyValues {SpriteRenderer=&self:NE_VC_Boost_Explosions003_7.NiagaraSpriteRendererProperties_0}}
    SubUV Animation Mode:ENiagara_SubUVLookupModeV2=::NewEnumerator1
    SubUV Curve=DataInterface() {propertyValues {Curve {keys [
        {interpMode=RCIM_Cubic, leaveTangentWeight=0.351, tangentMode=RCTM_User, time=0.25, value=0.2}
        {arriveTangent=-0.151, interpMode=RCIM_Cubic, leaveTangent=-0.151, tangentMode=RCTM_User, time=0.55, value=1}
       ]}}}
    UseEndFrame=false
    UseStartFrame=false
   )
   &ScaleColor(
    Scale Alpha=&MakeCustomFloatFromBool(
     Bool=$User.bIsFrontendPreview
     True Float=1
     False Float=&FloatFromCurve(FloatCurve=DataInterface() {propertyValues {Curve {keys [<at 0 1 RCIM_Cubic>, {arriveTangent=-1.5196, interpMode=RCIM_Cubic, leaveTangent=-1.5196, tangentMode=RCTM_User, time=1, value=0}]}}})
    )
    Scale RGB=/Niagara/DynamicInputs/TypeConversions/MakeCustomPositionFromBool(
     Bool:NiagaraBool=$User.bIsFrontendPreview
     True Position:NiagaraPosition=&VectorFromFloat(Value=&FloatFromCurve(
       FloatCurve=DataInterface() {propertyValues {Curve {keys [
           {arriveTangent=9.0725, interpMode=RCIM_Cubic, leaveTangent=9.0725, tangentMode=RCTM_User, time=0, value=0}
           {arriveTangent=0.3343, interpMode=RCIM_Cubic, leaveTangent=0.3343, tangentMode=RCTM_User, time=0.15, value=0.9847}
           {arriveTangent=-1.1302, interpMode=RCIM_Cubic, leaveTangent=4.3768, tangentMode=RCTM_Break, time=0.7, value=0.225}
           <at 1 10 RCIM_Cubic>
          ]}}}
       Scale Curve=3.5
      ))
     False Position:NiagaraPosition=&VectorFromFloat(Value=&FloatFromCurve(
       FloatCurve=DataInterface() {propertyValues {Curve {keys [
           {arriveTangent=0.0153, interpMode=RCIM_Cubic, leaveTangent=0.0153, tangentMode=RCTM_User, time=0.01, value=0.025}
           {arriveTangent=0.3343, interpMode=RCIM_Cubic, leaveTangent=0.3343, tangentMode=RCTM_User, time=0.07, value=1.5}
           {
            arriveTangent=-1.446
            arriveTangentWeight=0.1821
            interpMode=RCIM_Cubic
            leaveTangent=5.6015
            leaveTangentWeight=0.2535
            tangentMode=RCTM_Break
            time=0.5
            value=1
           }
           <at 1 2 RCIM_Cubic>
          ]}}}
       Scale Curve=7
      ))
    )
    ScaleA=true
    ScaleRGB=true
   ) {module=ScaleColor001}
  ]
  rendererClasses [&NiagaraSpriteRendererProperties]
  renderers [{rendererClass=&NiagaraSpriteRendererProperties}]
 }
 NE_VC_Boost_Ribbon_MAIN {
  emitterUpdateScript [
   &EmitterState(Life Cycle Mode=::NewEnumerator1, Loop Duration=1, Recalculate Duration Each Loop=true, UseLoopDelay=false)
   &VC_FilteredSocketCount()
   &VC_RateQualityHandler()
   &VC_SpawnRibbon_v1(
    Spawn Time=-1
    SpawnRate=&MakeCustomFloatFromBool(Bool=$User.bIsFrontendPreview, True Float=50, False Float=30)
   )
  ]
  particleSpawnScript [
   &VC_PaintedVariantsOverride(
    DefaultColor=<rgba 0.08 0.0651 0.024 1>
    R1 DefaultColor=<rgba 0.099 0.0809 0.0687 1.2>
    R2 DefaultColor:LinearColor=<rgba 0.4167 1 0 1>
    Use R1 Painted Overrides=true
    Use R2 Painted Overrides=true
   )
   &VC_InitializeRibbon_v1(
    Color=$Particles.Color
    Custom Offset=true
    Lifetime=&MakeCustomFloatFromBool(Bool=$User.bIsFrontendPreview, True Float=1.7, False Float=1.2)
    Local Vector Offset=&AddVectorToPosition(
     Position=&AddVectorToPosition(
      Position=&MakeCustomPositionFromBool(Bool:NiagaraBool=$User.bIsFrontendPreview, True Position:NiagaraPosition=<-45 0 0>, False Position:NiagaraPosition=<15 0 0>)
      Vector=$User.SingleBoostPos
     )
     Vector=&Multiply_VectorByFloat(Vector=$User.DTCarOffset, Float=1)
    )
    Number of Bones Per Boneflow group=0
    Offset Amount=1
    Ribbon Width=70
   )
   &AddVelocity(
    Velocity=&MakeCustomVectorFromBool(Bool=$User.bIsFrontendPreview, True Vector=<1000 0 0>, False Vector=<2000 0 0>)
    Velocity Speed Scale=-1
   ) {module=AddVelocity001}
   /SetVariables(Particles.COLOR2:LinearColor=<rgba 0.4167 1 0 1>) {enabled=false, isSetParameters=true}
  ]
  particleUpdateScript [
   &ParticleState()
   &InheritSourceMovement(
    Applied Position Scale=&Multiply_VectorByFloat(Vector=<1 1 1>, Float=1)
    ApplyForce=false
    ApplyPosition=true
    ApplyPositionFalloff=false
   )
   &SolveForcesAndVelocity(Clamp Velocity=false, Limit Acceleration=false)
   &VC_UpdateRibbon_v1()
   &ScaleColor(
    Scale Alpha=&FloatFromCurve(
     FloatCurve=DataInterface() {propertyValues {Curve {keys [{interpMode=RCIM_Cubic, tangentMode=RCTM_Break, time=0, value=1}, {interpMode=RCIM_Cubic, tangentMode=RCTM_Break, time=0.8, value=0}]}}}
     Scale Curve=2
    )
    Scale RGB=&Multiply_VectorByFloat(Vector=<1 1 1>, Float=7)
    ScaleA=true
    ScaleRGB=true
   )
   &DynamicMaterialParameters(
    Index 0 Param 1=4
    Index 0 Param 2=&MakeCustomFloatFromBool(Bool=$User.bIsFrontendPreview, True Float=7, False Float=7)
    Index 0 Param 3=&MakeCustomFloatFromBool(Bool=$User.bIsFrontendPreview, True Float=1, False Float=2)
    Module.Index0Param0Write=true
    Module.Index0Param1Write=true
    Module.Index0Param2Write=true
    Module.Index0Param3Write=true
   ) {enabled=false}
   &ScaleRibbonWidth(Ribbon Width Scale=&FloatFromCurve(FloatCurve=DataInterface() {propertyValues {Curve {keys [{interpMode=RCIM_Cubic, tangentMode=RCTM_User, time=0, value=1}, {interpMode=RCIM_Cubic, tangentMode=RCTM_User, time=1, value=0.5}]}}}))
   /SetVariables(
    Particles.ScaleColor_SmokeTrail:LinearColor=&MultiplyLinearColor(
     Linear Color=$Particles.R1 Color
     Scale Factor=&MakeVector4FromVectorAndFloat(
      "Vector (RGB)"=<1 1 1>
      "Float (Alpha)"=&FloatFromCurve(
       FloatCurve=DataInterface() {propertyValues {Curve {keys [{interpMode=RCIM_Cubic, tangentMode=RCTM_Break, time=0, value=1}, {interpMode=RCIM_Cubic, tangentMode=RCTM_Break, time=1.2, value=0}]}}}
       Scale Curve=1.5
      )
     )
    )
    Particles.Size_SmokeTrail:NiagaraFloat=&Multiply_Float(
     A=80
     B=&FloatFromCurve(FloatCurve=DataInterface() {propertyValues {Curve {keys [{interpMode=RCIM_Cubic, tangentMode=RCTM_User, time=0, value=1}, {interpMode=RCIM_Cubic, tangentMode=RCTM_User, time=1, value=0.5}]}}})
    )
   ) {isSetParameters=true}
   /SetVariables(Particles.ScaleColor_NoiseRibbon:LinearColor=&MultiplyLinearColor(
     Linear Color=$Particles.R2 Color
     Scale Factor=&MakeVector4FromVectorAndFloat(
      "Vector (RGB)"=<1 1 1>
      "Float (Alpha)"=&FloatFromCurve(FloatCurve=DataInterface() {propertyValues {Curve {keys [<at 0 1>,<at 0.8 0>]}}})
     )
    )) {isSetParameters=true}
  ]
  rendererClasses [&NiagaraRibbonRendererProperties]
  renderers [
   {rendererClass=&NiagaraRibbonRendererProperties}
   {rendererClass=&NiagaraRibbonRendererProperties, rendererIndex=1}
   {rendererClass=&NiagaraRibbonRendererProperties, rendererIndex=2}
  ]
 }
 NE_VC_Boost_Sprite_GLIMPSEAURA {
  emitterUpdateScript [
   &EmitterState(Inactive Response=::NewEnumerator1, Life Cycle Mode=::NewEnumerator1, Loop Duration=0.3167, UseLoopDelay=false)
   &SpawnBurst_Instantaneous(Spawn Group=1, Use Spawn Probability=false, UseLoopCountLimit=false) {module=SpawnBurst_Instantaneous001}
   &SpawnBurst_Instantaneous(
    Spawn Count=&MakeCustomIntFromBool(Bool=$User.bIsFrontendPreview, True INT=1, False INT=1)
    Use Spawn Probability=false
    UseLoopCountLimit=false
   )
   &VC_FilteredSocketCount()
   &SpawnRate(SpawnRate=5, Use Spawn Probability=false) {enabled=false}
   &SpawnBurst_Instantaneous(Spawn Time=0.3167, Use Spawn Probability=false, UseLoopCountLimit=false) {module=SpawnBurst_Instantaneous002}
  ]
  particleSpawnScript [
   &VC_PaintedVariantsOverride() {enabled=false}
   /SetVariables(Particles.RENDERTAG:NiagaraInt32=&SpawnGroup()) {isSetParameters=true}
   &InitializeParticle(
    Color=<rgba 0.5 1.5 0 1>
    Lifetime=&VC_FloatBySpawnGroup(
     SpawnGroup or ExecIndex:NiagaraInt32=0
     CustomInt:NiagaraInt32=0
     MaximumSpawnGroups:NiagaraInt32=2
     Float SG 0:NiagaraFloat=0.1
     Float SG 1:NiagaraFloat=&RandomRangeFloat(Minimum=0.3, Maximum=0.25)
     Float SG 2:NiagaraFloat=0
     Float SG 3:NiagaraFloat=0
     Float SG 4:NiagaraFloat=0
     Float SG 5:NiagaraFloat=0
    )
    Position Offset=&AddVectorToPosition(
     Position=&AddVectorToPosition(
      Position=&VC_VectorBySpawnGroup(
       SpawnGroup or ExecIndex:NiagaraInt32=0
       CustomInt:NiagaraInt32=0
       MaximumSpawnGroups:NiagaraInt32=2
       Vector Group 0:Vector3f=<-40 0 0>
       Vector Group 1:Vector3f=&RandomRangeVector(Minimum=<-5 150 0>, Maximum=<0 150 0>)
       Vector Group 2:Vector3f=<0 0 0>
       Vector Group 3:Vector3f=<0 0 0>
       Vector Group 4:Vector3f=<0 0 0>
       Vector Group 5:Vector3f=<0 0 0>
      )
      Vector=$User.SingleBoostPos
     )
     Vector=&Multiply_VectorByFloat(Vector=$User.DTCarOffset, Float=1)
    )
    Position Offset Coordinate Space=::Local
    Sprite Rotation Angle=90
    Sprite Rotation Mode=::NewEnumerator0
    Sprite Size=&SelectVector2DFromArray(
     Array Sampling Mode=::NewEnumerator0
     Direct Array Index=?
     Vector 2D Selection Array=DataInterface() {propertyValues {FloatData [<100 70>, <100 -70>]}}
     Random Seed=0
    )
    Sprite Size Mode=::NewEnumerator4
    UsePositionOffset=true
    Write Material Random=true
   )
   &PartitionParticles(Number Of Partitions=1) {enabled=false}
   &SkeletalMeshLocation(
    Bone / Socket Index=$Particles.PartitionParticles.Partition
    Bone / Socket Sampling Mode=::NewEnumerator5
    MaskSpawnGroup=false
    Mesh Sampling Type=::NewEnumerator6
    Sampled Position Offset=&AddVectorToPosition(
     Position=&AddVectorToPosition(Position=<-50 0 0>, Vector=$User.SingleBoostPos)
     Vector=&Multiply_VectorByFloat(Vector=$User.DTCarOffset, Float=1)
    )
    Skeletal Mesh=$System.SkeletalMesh
   ) {enabled=false}
   &ApplyInitialForces()
   &AddVelocity(
    Cone Angle=0
    Cone Axis=<-1 0 0>
    Module.UseConeFalloff=false
    Module.UseVelDistribution=false
    Override Seed=false
    Velocity Mode=::NewEnumerator2
    Velocity Speed=&VC_FloatBySpawnGroup(
     SpawnGroup or ExecIndex:NiagaraInt32=0
     CustomInt:NiagaraInt32=0
     MaximumSpawnGroups:NiagaraInt32=2
     Float SG 0:NiagaraFloat=200
     Float SG 1:NiagaraFloat=&RandomRangeFloat(Minimum=400, Maximum=100)
     Float SG 2:NiagaraFloat=0
     Float SG 3:NiagaraFloat=0
     Float SG 4:NiagaraFloat=0
     Float SG 5:NiagaraFloat=0
    )
   )
   &AlignSpriteToMeshOrientation(
    Mesh Orientation Relative Sprite Alignment Vector=&VC_VectorBySpawnGroup(
     SpawnGroup or ExecIndex:NiagaraInt32=0
     CustomInt:NiagaraInt32=0
     MaximumSpawnGroups:NiagaraInt32=2
     Vector Group 0:Vector3f=<1 0 0>
     Vector Group 1:Vector3f=<1 0 0>
     Vector Group 2:Vector3f=<0 0 0>
     Vector Group 3:Vector3f=<0 0 0>
     Vector Group 4:Vector3f=<0 0 0>
     Vector Group 5:Vector3f=<0 0 0>
    )
    Mesh Orientation Relative Sprite Facing Vector:Vector3f=<0 1 1>
    SetAlignment=true
    SetFacing=false
   )
   &CameraOffset(Camera Offset Amount=&VC_FloatBySpawnGroup(
     SpawnGroup or ExecIndex:NiagaraInt32=0
     CustomInt:NiagaraInt32=0
     MaximumSpawnGroups:NiagaraInt32=2
     Float SG 0:NiagaraFloat=50
     Float SG 1:NiagaraFloat=100
     Float SG 2:NiagaraFloat=0
     Float SG 3:NiagaraFloat=0
     Float SG 4:NiagaraFloat=0
     Float SG 5:NiagaraFloat=0
    ))
  ]
  particleUpdateScript [
   &ParticleState()
   &ScaleSpriteSize(
    Scale Factor=&MaskVector2dBySpawnGroup(
     Masked Vector2D=&Vector2DFromCurve(
      Vector2Curve=DataInterface() {propertyValues {
        XCurve {keys [{arriveTangent=3.6376, interpMode=RCIM_Cubic, leaveTangent=3.6376, tangentMode=RCTM_User, time=0, value=0}, <at 0.5 1 RCIM_Cubic>]}
        YCurve {keys [
          {arriveTangent=0.1453, interpMode=RCIM_Cubic, leaveTangent=0.1453, leaveTangentWeight=1.1406, tangentMode=RCTM_User, time=0, value=0}
          <at 0.5 0.8 RCIM_Cubic>
         ]}
       }}
      ScaleCurve=<1 1>
     )
     Passthrough Vector2D=&Vector2DFromCurve(
      Vector2Curve=DataInterface() {propertyValues {
        XCurve {keys [{arriveTangent=6.7699, interpMode=RCIM_Cubic, leaveTangent=6.7699, tangentMode=RCTM_User, time=0, value=0.2}, <at 0.5 2 RCIM_Cubic>]}
        YCurve {keys [
          {arriveTangent=5.0877, interpMode=RCIM_Cubic, leaveTangent=5.0877, leaveTangentWeight=1.1406, tangentMode=RCTM_User, time=0, value=0}
          {arriveTangent=2, interpMode=RCIM_Cubic, leaveTangent=2, time=0.25, value=1}
          <at 0.5 1 RCIM_Cubic>
         ]}
       }}
      ScaleCurve=<1 1>
     )
     Spawn Group Mask=0
    )
    Scale Sprite Size Mode=::NewEnumerator2
   ) {module=ScaleSpriteSize001}
   &SolveForcesAndVelocity(Clamp Velocity=false, Limit Acceleration=false)
   &ScaleColor(
    Scale Alpha=&FloatFromCurve(
     FloatCurve=DataInterface() {propertyValues {Curve {keys [
         {arriveTangent=0.3151, interpMode=RCIM_Cubic, leaveTangent=0.3151, tangentMode=RCTM_User, time=0, value=0}
         {interpMode=RCIM_Cubic, tangentMode=RCTM_User, time=0.5, value=1}
         {arriveTangent=0.0159, tangentMode=RCTM_User, time=1, value=0}
        ]}}}
     Scale Curve=&VC_FloatBySpawnGroup(
      SpawnGroup or ExecIndex:NiagaraInt32=0
      CustomInt:NiagaraInt32=0
      MaximumSpawnGroups:NiagaraInt32=2
      Float SG 0:NiagaraFloat=1
      Float SG 1:NiagaraFloat=&FloatFromCurve(FloatCurve=DataInterface() {propertyValues {Curve {keys [<at 0 1>,<at 1 0.5>]}}})
      Float SG 2:NiagaraFloat=0
      Float SG 3:NiagaraFloat=0
      Float SG 4:NiagaraFloat=0
      Float SG 5:NiagaraFloat=0
     )
    )
    ScaleA=true
    ScaleRGB=true
   )
  ]
  rendererClasses [&NiagaraSpriteRendererProperties]
  renderers [
   {rendererClass=&NiagaraSpriteRendererProperties}
   {rendererClass=&NiagaraSpriteRendererProperties, rendererIndex=1}
  ]
 }
 NE_VC_PUMPKIN_FRONT {
  emitterUpdateScript [
   &EmitterState(
    Life Cycle Mode=::NewEnumerator1
    Loop Duration=&MakeCustomFloatFromBool(Bool=$User.bIsFrontendPreview, True Float=0.25, False Float=0.05)
    Recalculate Duration Each Loop=true
    UseLoopDelay=false
   )
   &SpawnBurst_Instantaneous(
    Spawn Count=&MakeCustomIntFromBool(Bool=$User.bIsFrontendPreview, True INT=1, False INT=1)
    Use Spawn Probability=false
    UseLoopCountLimit=false
   )
   &VC_FilteredSocketCount() {enabled=false}
   &SpawnRate(SpawnRate=16, Use Spawn Probability=false) {enabled=false}
  ]
  particleSpawnScript [
   &VC_PaintedVariantsOverride(
    Crimson=<rgba 1 0.03 0.0462 1>
    DefaultColor=&MakeCustomLinearColorFromBool(Bool=$User.bIsFrontendPreview, True Color=<rgba 0.2 0.0267 0 1>, False Color=<rgba 0.65 0.1083 0 1>)
    R1 DefaultColor=<rgba 0 1 0 1>
    R2 DefaultColor:LinearColor=<rgba 1 0.7 0 1>
    R3 DefaultColor:LinearColor=<rgba 0.05 0.0008 0 1>
    R4 DefaultColor:LinearColor=&MakeCustomLinearColorFromBool(Bool=$User.bIsFrontendPreview, True Color=<rgba 0.2 0.0267 0 1>, False Color=<rgba 0.65 0.1083 0 1>)
    Use R1 Painted Overrides=true
    Use R2 Painted Overrides=true
    Use R3 Painted Overrides=true
    Use R4 Painted Overrides=true
   )
   &InitializeParticle(
    Color=$Particles.Color
    Lifetime=&MakeCustomFloatFromBool(Bool=$User.bIsFrontendPreview, True Float=&RandomRangeFloat(Minimum=0.6, Maximum=0.7), False Float=0.1)
    Mesh Scale Mode=::NewEnumerator1
    Mesh Uniform Scale Max=1
    Position Offset=&AddVectorToPosition(
     Position=&AddVectorToPosition(Position=<-75 0 10>, Vector=$User.SingleBoostPos)
     Vector=&Multiply_VectorByFloat(Vector=$User.DTCarOffset, Float=1)
    )
    Position Offset Coordinate Space=::Local
    Sprite Rotation Mode=::NewEnumerator2
    Sprite Size Min=<2 5>
    Sprite Size Mode=::NewEnumerator2
    UsePositionOffset=true
    Write Material Random=true
   )
   &PartitionParticles(Number Of Partitions=$Emitter.PortCount) {enabled=false}
   &SkeletalMeshLocation(
    Bone / Socket Index=1
    Bone / Socket Sampling Mode=::NewEnumerator5
    MaskSpawnGroup=false
    Mesh Sampling Type=::NewEnumerator6
    Sampled Position Offset=&AddVectorToPosition(
     Position=&AddVectorToPosition(Position=<120 0 -30>, Vector=$User.SingleBoostPos)
     Vector=&Multiply_VectorByFloat(Vector=$User.DTCarOffset, Float=1)
    )
    Skeletal Mesh=$System.SkeletalMesh
   ) {enabled=false}
   &AddVelocity(
    Cone Angle=0
    Cone Axis=&RandomRangeVector(Minimum=<-1 -0.15 0.2>, Maximum=<-1 0.1 0.05>)
    Module.UseConeFalloff=false
    Module.UseVelDistribution=false
    Override Seed=false
    Velocity Mode=::NewEnumerator2
    Velocity Speed=&SelectFloatFromArray(
     Array Sampling Mode=::NewEnumerator0
     Direct Array Index=?
     Float Selection Array=DataInterface() {propertyValues {FloatData [700,700]}}
     Random Seed=0
    )
   )
   &ShapeLocation(MaskSpawnGroup=false, Override Seed=false, Sphere Radius=35, UseOwnerScale=true)
   &InitialMeshOrientation(Rotation=<1 1 1>, Use Rotation Vector=false)
  ]
  particleUpdateScript [
   &ParticleState()
   &UpdateMeshOrientation(Rotation Rate=2)
   &ScaleVelocity(
    Coordinate Space:ENiagaraCoordinateSpace=::Simulation
    Velocity Scale=&VectorFromCurve(
     VectorCurve=DataInterface() {propertyValues {
       XCurve {keys [<at 0 1>,<at 1 1>]}
       YCurve {keys [<at 0 1>,<at 1 1>]}
       ZCurve {keys [
         {arriveTangent=-14.771, interpMode=RCIM_Cubic, leaveTangent=-14.771, tangentMode=RCTM_User, time=0, value=0}
         {arriveTangentWeight=0.9748, interpMode=RCIM_Cubic, tangentMode=RCTM_User, time=1, value=2}
        ]}
      }}
     Scale Curve=&RandomRangeVector(Minimum=<1.15 1 -1>, Maximum=<1.3 1 -1.2>)
    )
   )
   &InheritSourceMovement(
    Applied Position Scale=&Multiply_VectorByFloat(
     Vector=<1 1 1>
     Float=&FloatFromCurve(FloatCurve=DataInterface() {propertyValues {Curve {keys [
          {arriveTangent=-0.5211, interpMode=RCIM_Cubic, leaveTangent=-0.5211, tangentMode=RCTM_User, time=0, value=1}
          {arriveTangentWeight=0.3344, interpMode=RCIM_Cubic, tangentMode=RCTM_User, time=1, value=0}
         ]}}})
    )
    ApplyForce=false
    ApplyPosition=true
    ApplyPositionFalloff=false
   )
   &Collision(
    Age Colliding Particles=false
    Analytical Collision Plane Position 1:Vector3f=&RandomRangeVector(Minimum=<90 0 0>, Maximum=<90 0 20>)
    Analytical Collision Plane Space:ENiagaraCoordinateSpace=::Local
    CPU Collision Type=::NewEnumerator3
    Enable Rest State=false
    KillOnCollision=true
    Module.EnableMaxCollisionCount=false
    Module.Randomize Collision Normal Vector=false
    Perform 2 Analytical Collision Queries=false
    Restitution=0
   ) {enabled=false}
   &Drag(
    Drag=2
    Rotational Drag:NiagaraFloat=&FloatFromCurve(FloatCurve=DataInterface() {propertyValues {Curve {keys [<at 0 1 RCIM_Cubic>,<at 1 0 RCIM_Cubic>]}}}, Scale Curve=5)
    Use Linear Drag=true
    Use Rotational Drag=true
   )
   &SolveForcesAndVelocity(Clamp Velocity=false, Limit Acceleration=false)
   &GenerateDeathEvent(Use Event Probability=false)
   /SetVariables(
    Particles.GreenColor:LinearColor=&MultiplyLinearColor(
     Linear Color=$Particles.R1 Color
     Scale Factor=&MakeVector4FromVectorAndFloat(
      "Vector (RGB)"=&VectorFromFloat(Value=&Waveform("Waveform [1]"=::NewEnumerator0, "[1] Phase Scale"=0.4, "[1] Frequency"=1, "[1] Amplitude Min/Max"=<1 3>))
      "Float (Alpha)"=1
     )
    )
    Particles.Linear Color=&MultiplyLinearColor(
     Linear Color=$Particles.R3 Color
     Scale Factor=&MakeVector4FromVectorAndFloat("Vector (RGB)"=<3 3 3>, "Float (Alpha)"=0.5)
    )
   ) {isSetParameters=true}
   &DynamicMaterialParameters(
    Index 1 Param 2=1
    Module.Index1Float0Write=true
    Module.Index1Float1Write=true
    Module.Index1Float2Write=true
    Module.Index1Float3Write=true
    Write Parameter Index 0=::NewEnumerator0
    Write Parameter Index 1=::NewEnumerator1
   ) {enabled=false}
   /SetVariables(Particles.FireSprite:LinearColor=&MultiplyLinearColor(
     Linear Color=$Particles.R4 Color
     Scale Factor=&MakeVector4FromVectorAndFloat("Vector (RGB)"=&VectorFromFloat(Value=6), "Float (Alpha)"=1)
    )) {isSetParameters=true}
  ]
  rendererClasses [&NiagaraMeshRendererProperties, &NiagaraSpriteRendererProperties]
  renderers [
   {rendererClass=&NiagaraMeshRendererProperties, rendererIndex=0}
   {rendererClass=&NiagaraMeshRendererProperties, rendererIndex=1}
   {rendererClass=&NiagaraMeshRendererProperties, rendererIndex=2}
   {rendererClass=&NiagaraMeshRendererProperties, rendererIndex=3}
   {rendererClass=&NiagaraMeshRendererProperties, rendererIndex=4}
   {rendererClass=&NiagaraMeshRendererProperties, rendererIndex=5}
   {rendererClass=&NiagaraSpriteRendererProperties, rendererIndex=6}
  ]
 }
}
summary {
 systemName=NS_Sticky_02
 emitters [
  {emitterName=NE_VC_Boost_Ribbon_MAIN, rendererClasses [&NiagaraRibbonRendererProperties]}
  {
   emitterName=NE_VC_PUMPKIN_FRONT
   rendererClasses [&NiagaraMeshRendererProperties, &NiagaraSpriteRendererProperties]
  }
  {emitterName=NE_VC_Boost_Sprite_GLIMPSEAURA, rendererClasses [&NiagaraSpriteRendererProperties]}
  {
   emitterName=NE_VC_Boost_Core_Aura
   rendererClasses [&NiagaraMeshRendererProperties, &NiagaraSpriteRendererProperties]
  }
  {emitterName=NE_VC_Boost_Explosions003, rendererClasses [&NiagaraSpriteRendererProperties]}
 ]
 userVariables [
  User.bIsFrontendPreview {type=NiagaraBool}
  User.bIsPainted {type=NiagaraBool}
  User.ColorVariantID {type=NiagaraInt32}
  User.DTCarOffset {type=Vector3f, description="", defaultValue {struct=Vector3f, value=<1 1 1>}}
  User.PaintedColor {type=LinearColor, description="", defaultValue {struct=LinearColor, value=<rgba 1 1 1 1>}}
  User.PaintedScalar {type=NiagaraFloat, defaultValue=1, description=""}
  User.PortCount {type=NiagaraInt32, defaultValue=1}
  User.SingleBoostPos {type=Vector3f, description="", defaultValue {struct=Vector3f, value=<0 0 0>}}
 ]
}
